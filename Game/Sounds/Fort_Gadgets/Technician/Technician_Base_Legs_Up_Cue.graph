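# Sound cue for the Technician gadget base's legs-up sound in the gadgets sound folder.
# A single wave player plays the Fort_BASE_Legs_Up wave through a modulator node left at default settings.
# It uses the gadgets sound class and overrides its named Siphon attenuation with inline logarithmic falloff.
# A concurrency override prevents new instances from starting once the limit is reached.
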SoundCue /Game/Sounds/Fort_Gadgets/Technician/Technician_Base_Legs_Up_Cue
root=Modulator_0
nodes [
 {id=Modulator_0, children [WavePlayer_3]}
 {id=WavePlayer_3, properties {soundWaveAssetPtr=&/Fort_BASE_Legs_Up}}
]
properties {
 attenuationSettings=/Game/Sounds/Attenuation/Fort_Siphon_Att
 bOverrideAttenuation=true
 bOverrideConcurrency=true
 maxDistance=1700
 soundClassObject=/Game/Sounds/FortSoundClassesAndModes/Fort_Gadgets_StW
 volumeMultiplier=1
 attenuationOverrides {distanceAlgorithm=Logarithmic, falloffDistance=1200, attenuationShapeExtents {x=500}}
 concurrencyOverrides {resolutionRule=PreventNew}
}
