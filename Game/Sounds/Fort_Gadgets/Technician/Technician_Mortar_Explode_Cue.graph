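# Sound cue for the Technician gadget's mortar explosion, built from electrical zap sounds rather than blast samples.
# A Random node picks one of four equally weighted husk electrocution waves and feeds a Modulator that varies pitch and volume per play.
# It uses close-range explosion attenuation and the large explosives sound class, with a concurrency override that caps simultaneous voices by stopping the oldest.
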SoundCue /Game/Sounds/Fort_Gadgets/Technician/Technician_Mortar_Explode_Cue
root=Modulator_2
nodes [
 {id=Modulator_2, children [Random_1], properties {pitchMin=0.85, volumeMax=1.2, volumeMin=0.9}}
 {id=Random_1, children [WavePlayer_2,WavePlayer_3,WavePlayer_5,WavePlayer_6], properties {weights [1,1,1,1]}}
 {id=WavePlayer_2, properties {soundWaveAssetPtr=/Game/Sounds/Fort_Gadgets/Electrical_Zaps/husk_electrocute_03}}
 {id=WavePlayer_3, properties {soundWaveAssetPtr=/Game/Sounds/Fort_Gadgets/Electrical_Zaps/husk_electrocute_05}}
 {id=WavePlayer_5, properties {soundWaveAssetPtr=/Game/Sounds/Fort_Gadgets/Electrical_Zaps/husk_electrocute_06}}
 {id=WavePlayer_6, properties {soundWaveAssetPtr=/Game/Sounds/Fort_Gadgets/Electrical_Zaps/husk_electrocute_08}}
]
properties {
 attenuationSettings=/Game/Sounds/Attenuation/Fort_Explosion_Close_StW_Att
 bOverrideConcurrency=true
 maxDistance=3500
 soundClassObject=/Game/Sounds/FortSoundClassesAndModes/Fort_Explosives_Large_PSM
 volumeMultiplier=0.5236
 concurrencyOverrides {maxCount=6, resolutionRule=StopOldest}
}
