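# Sound cue for the looping working hum of the Technician mortar gadget.
# A single looping wave player plays the Portal_Hum1 wave with no other mixing or modulation.
# It uses the gadget-working attenuation and gadget sound class, with overridden concurrency that caps simultaneous instances by stopping the farthest before blocking new ones.
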SoundCue /Game/Sounds/Fort_Gadgets/Technician/Technician_Mortar_Loop_Cue
root=WavePlayer_11
nodes [{id=WavePlayer_11, properties {bLooping=true, soundWaveAssetPtr=&/Portal_Hum1}}]
properties {
 attenuationSettings=/Game/Sounds/Attenuation/Fort_GadgetWorking_Att
 bOverrideConcurrency=true
 maxDistance=3100
 soundClassObject=/Game/Sounds/FortSoundClassesAndModes/Fort_Gadgets_StW
 volumeMultiplier=1
 concurrencyOverrides {maxCount=3, resolutionRule=StopFarthestThenPreventNew}
}
